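# Physics asset for the Glow Bro backpack cosmetic, defining collision and joints for its dangling left and right straps.
# Each strap is a chain of three box bodies linked to the pack_mid capsule, with wide swing limits and tight twist limits.
# An unconstrained pelvis_collision capsule stands apart from the chains, apparently as a collider for the straps against the wearer's hips.
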PhysicsAsset /BRCosmetics/Accessories/FORT_Backpacks/M_MED_Glow_Bro_Backpack/Meshes/M_MED_Glow_Bro_Pack_PhysicsAsset
bodies {
 Dyn_strap_01_l [{shapeType=Box, extentX=1.1, extentY=7.5, extentZ=7.5, center=<4 0 0>, rotation=<rot 90 0 0>}]
 Dyn_strap_01_r [{shapeType=Box, extentX=1.1, extentY=7.5, extentZ=7.5, center=<-4 0 0>, rotation=<rot 90 0 0>}]
 Dyn_strap_02_l [{shapeType=Box, extentX=1.1, extentY=7.5, extentZ=7.5, center=<4 0 0>, rotation=<rot 90 0 0>}]
 Dyn_strap_02_r [{shapeType=Box, extentX=1.1, extentY=7.5, extentZ=7.5, center=<-4 0 0>, rotation=<rot 90 0 0>}]
 Dyn_strap_03_l [{shapeType=Box, extentX=1.1, extentY=7.5, extentZ=7.5, center=<4 0 0>, rotation=<rot 90 0 0>}]
 Dyn_strap_03_r [{shapeType=Box, extentX=1.1, extentY=7.5, extentZ=7.5, center=<-4 0 0>, rotation=<rot 90 0 0>}]
 pack_mid [{shapeType=Capsule, length=10.5482, radius=12.3439, center=<0.0002 15.6324 -14.7193>, rotation=<rot 0 -89.9998 89.9992>}]
 pelvis_collision [{shapeType=Capsule, length=9.3457, radius=10.7641, center=<-5.0596 -2.414 0.0175>}]
}
constraints [
 {bone1Name=Dyn_strap_02_r, bone2Name=Dyn_strap_01_r, swing1LimitDegrees=60, swing2LimitDegrees=60, twistLimitDegrees=10}
 {bone1Name=Dyn_strap_03_r, bone2Name=Dyn_strap_02_r, swing1LimitDegrees=60, swing2LimitDegrees=60, twistLimitDegrees=10}
 {bone1Name=Dyn_strap_02_l, bone2Name=Dyn_strap_01_l, swing1LimitDegrees=60, swing2LimitDegrees=60, twistLimitDegrees=10}
 {bone1Name=Dyn_strap_03_l, bone2Name=Dyn_strap_02_l, swing1LimitDegrees=60, swing2LimitDegrees=60, twistLimitDegrees=10}
 {bone1Name=Dyn_strap_01_r, bone2Name=pack_mid, swing1LimitDegrees=60, swing2LimitDegrees=60, twistLimitDegrees=5}
 {bone1Name=Dyn_strap_01_l, bone2Name=pack_mid, swing1LimitDegrees=60, swing2LimitDegrees=60, twistLimitDegrees=5}
]
