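# Animation Blueprint for the Glow Bro backpack accessory, adding secondary motion by switching between two simulation methods by LOD.
# Each update copies the player's pelvis socket transform from the Head part mesh onto a pelvis_collision bone, feeding a RigidBody simulation with an override physics asset.
# At LOD 2 and above, TrailController blends to cheaper chained trail controllers on the left and right straps, constrained by planar limits on pack_mid.
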AnimBlueprint /BRCosmetics/Accessories/FORT_Backpacks/M_MED_Glow_Bro_Backpack/Meshes/M_MED_Glow_Bro_Pack_Skeleton_AnimBP
target_skeleton=&/M_MED_Glow_Bro_Pack_Skeleton
graphs [
 AnimGraph {nodes [
   n1 = Trail controller - Bone: Dyn_strap_03_r!(ComponentPose=Local To Component(LocalPose=Use cached pose 'InputPose'()), Alpha=1, FakeVelocity:Vector="0,0,0") -> (Pose) {
    bActorSpaceFakeVel=true
    bInvertChainBoneAxis=true
    bShowBaseMotion=false
    bShowLimit=false
    bUsePlanarLimit=true
    chainLength=3
    lODThreshold=4
    maxDeltaTime=0.0333
    relaxationSpeedScale=2
    baseJoint {boneName=pack_mid}
    planarLimits [
     {drivingBone {boneName=pack_mid}, planeTransform {location=<0 -2.0567 0>, rotation=<rot 0 0 -90>, scale=<1 1 1>}}
     {drivingBone {boneName=pack_mid}, planeTransform {location=<0 0 -10>, rotation=<rot 0 0 -150>, scale=<1 1 1>}}
    ]
    rotationLimits [{limitMax=<180 180 180>}, {limitMax=<180 180 180>}, {limitMax=<180 180 180>}]
    rotationOffsets [<0 0 0>, <0 0 0>, <0 0 0>]
    trailBone {boneName=Dyn_strap_03_r}
   }
   n2 = Trail controller - Bone: Dyn_strap_03_l!(ComponentPose=@n1, Alpha=1, FakeVelocity:Vector="0,0,0") -> (Pose) {
    bActorSpaceFakeVel=true
    bShowBaseMotion=false
    bShowLimit=false
    bUsePlanarLimit=true
    chainLength=3
    lODThreshold=4
    maxDeltaTime=0.0333
    relaxationSpeedScale=2
    baseJoint {boneName=pack_mid}
    planarLimits [
     {drivingBone {boneName=pack_mid}, planeTransform {location=<0 -2.0567 0>, rotation=<rot 0 0 -90>, scale=<1 1 1>}}
     {drivingBone {boneName=pack_mid}, planeTransform {location=<0 0 -10>, rotation=<rot 0 0 -150>, scale=<1 1 1>}}
    ]
    rotationLimits [{limitMax=<180 180 180>}, {limitMax=<180 180 180>}, {limitMax=<180 180 180>}]
    rotationOffsets [<0 0 0>, <0 0 0>, <0 0 0>]
    trailBone {boneName=Dyn_strap_03_l}
   }
   n3 = "Transform (Modify) Bone - Bone: pelvis_collision"(
    ComponentPose=Local To Component(LocalPose=Use cached pose 'InputPose'())
    Alpha=1
    Translation=Get pelvis trans() -> (pelvis trans:Vector)
    Rotation=Get pelvis rot() -> (pelvis rot:Rotator)
    Scale="1,1,1"
   ) {
    rotationMode=BMM_Replace
    rotationSpace=BCS_WorldSpace
    scaleSpace=BCS_BoneSpace
    translationMode=BMM_Replace
    translationSpace=BCS_WorldSpace
    alphaBoolBlend {blendOption=Linear}
    alphaScaleBiasClamp {bInterpResult=true, interpSpeedDecreasing=100, interpSpeedIncreasing=100}
    boneToModify {boneName=pelvis_collision}
   }
   n4 = RigidBody(ComponentPose=@n3) {
    bClampLinearTranslationLimitToRefPose=true
    bOverrideWorldGravity=true
    lODThreshold=1
    overridePhysicsAsset=&/M_MED_Glow_Bro_Pack_PhysicsAsset
    alphaBoolBlend {blendOption=Linear}
    componentAppliedLinearAccClamp=<3000 3000 3000>
    componentLinearAccScale=<0.5 0.3 0.3>
    componentLinearVelScale=<0.5 0.3 0.3>
    overrideWorldGravity {z=-980}
   }
   Output Pose(Result=Blend Poses by bool(
     BlendPose_0=Component To Local(ComponentPose=@n2)
     BlendPose_1=Component To Local(ComponentPose=@n4)
     bActiveValue=Get TrailController() -> (TrailController:bool)
    ) {blendType=Linear}) {layerGroup=None}
   AnimGraphNode_SubInput_0 = Input Pose()
   Save cached pose 'InputPose'(Pose=@AnimGraphNode_SubInput_0) {cacheName=InputPose}
  ]}
 EventGraph {nodes [
   Event BlueprintUpdateAnimation()
   n1 = Get PlayerPawn() -> (PlayerPawn:Fort Player Pawn Athena Object Reference)
   > Is Valid(InputObject=@n1)
   n2 = BreakTransform(InTransform=GetSocketTransform(self=GetSkeletalMeshForPartType(self=@n1, PartType=Head), InSocketName=pelvis, TransformSpace=RTS_World))
   > Set pelvis trans(pelvis trans:Vector=@n2) -> (Output_Get:Vector)
   > Set pelvis rot(pelvis rot:Rotator=@n2.Rotation) -> (Output_Get:Rotator)
   > n3 = Branch(Condition="integer >= integer"!(A:int=Get CurrentLOD(), B:int=2) -> (ReturnValue:bool) {is=title K2Node_CallFunction})
   > Set TrailController(TrailController:bool=true) -> (Output_Get:bool)
   Set TrailController(execute=@n3.else, TrailController=false)
   Event BlueprintInitializeAnimation()
   n4 = TryGetPawnOwner()
   > Is Valid(InputObject=@n4)
   > n5 = Cast To FortPlayerPawnAthena(Object=@n4)
   > Set PlayerPawn(PlayerPawn:Fort Player Pawn Athena Object Reference=@n5) -> (Output_Get:Fort Player Pawn Athena Object Reference)
   TryGetPawnOwner()
  ]}
]
variables [
 pelvis rot {type=Rotator}
 pelvis trans {type=Vector}
 PlayerPawn {type=FortPlayerPawnAthena}
 TrailController {type=boolean}
]
